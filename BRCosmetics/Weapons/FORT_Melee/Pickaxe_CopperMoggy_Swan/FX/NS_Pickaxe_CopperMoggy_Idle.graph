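# Idle effect for the CopperMoggy pickaxe cosmetic, adding a persistent light and glow.
# The Light emitter drives a light renderer, while Three_Way_Glow bursts three partitioned sprites, each with its own offset, size, orientation and visibility tag across three renderers.
# Both fade in over emitter age, and the sprites pulse over looping lifetimes, scaled differently for gameplay and front end.
# A variant color selector tints both emitters with one of three colors.
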NiagaraSystem /BRCosmetics/Weapons/FORT_Melee/Pickaxe_CopperMoggy_Swan/FX/NS_Pickaxe_CopperMoggy_Idle
dependencies {
 usedDataInterfaces [&NiagaraDataInterfaceCurve, &NiagaraDataInterfaceSocketReader]
 usedDynamicInputs []
 usedModules [
  &self:SystemSpawnScript.SystemScriptSource.SystemScriptGraph.NiagaraNodeAssignment_0.SetVariables_#1
  &SystemState
  &FN_VariantManager
 ]
}
emitters {
 Light {
  emitterSpawnScript [/SetVariables(Emitter.Position:NiagaraPosition=<0 -18 0>) {isSetParameters=true}]
  emitterUpdateScript [
   &EmitterState(
    Inactive Response=::NewEnumerator1
    Life Cycle Mode=::NewEnumerator1
    Loop Behavior=::NewEnumerator1
    Loop Duration Mode=::NewEnumerator1
    UseLoopDelay=false
   )
   &SpawnBurst_Instantaneous(Use Spawn Probability=false, UseLoopCountLimit=false)
  ]
  particleSpawnScript [&InitializeParticle(Lifetime=0, Position Offset=<-6.5 0 40.5>, UsePositionOffset=true, Write Material Random=true)]
  particleUpdateScript [
   &ParticleState(Kill Particles When Lifetime Has Elapsed=false)
   &Light_Attributes(
    Color=1
    Light Radius=50
    Radius=1
    Scale Alpha:NiagaraFloat=&FloatFromCurve(
     FloatCurve=DataInterface() {propertyValues {Curve {
        defaultValue=1
        keys [{interpMode=RCIM_Cubic, tangentMode=RCTM_User, time=0, value=0}, {interpMode=RCIM_Cubic, tangentMode=RCTM_User, time=1, value=1}]
       }}}
     CurveIndex=$Emitter.Age
    )
    Scale RGB:Vector3f=&MakeVectorFromLinearColorRGB(Color=&VariantColorSelector(Variant Amount=1, Original Color=<rgba 1 0.02 0.02 1>, Variant 1 Color=<rgba 0 1 0.6462 1>, Variant 2 Color=<rgba 1 0.5659 0 1>))
    ScaleA=true
    ScaleRGB=true
   )
  ]
  rendererClasses [&NiagaraLightRendererProperties]
  renderers [{rendererClass=&NiagaraLightRendererProperties, rendererIndex=0}]
 }
 Three_Way_Glow {
  emitterUpdateScript [
   &EmitterState(
    Inactive Response=::NewEnumerator1
    Life Cycle Mode=::NewEnumerator1
    Loop Behavior=::NewEnumerator1
    Loop Duration Mode=::NewEnumerator1
    UseLoopDelay=false
   )
   &SpawnBurst_Instantaneous(Spawn Count=3, Use Spawn Probability=false, UseLoopCountLimit=false)
  ]
  particleSpawnScript [
   &PartitionParticles(Number Of Partitions=3)
   &InitializeParticle(
    Lifetime=1.5
    Position Offset=&MergeHelper_VectorSwitch(0=<-6.5 0 40.5>, 1=<-16 0 33>, 2=<-9 0 46>, 3=?, Index=$Particles.PartitionParticles.Partition, Num=::NewEnumerator1)
    Sprite Rotation Mode=::NewEnumerator0
    Sprite Size=&MergeHelper_Vector2Switch(0=<59 58>, 1=<6 60>, 2=<6 75>, 3=?, 4=?, Index=$Particles.PartitionParticles.Partition, Num=::NewEnumerator1)
    Sprite Size Mode=::NewEnumerator4
    UsePositionOffset=true
    Write Material Random=true
   )
   &SpriteFacingAndAlignment(
    Sprite Alignment=&MergeHelper_VectorSwitch(0=<0 0 1>, 1=<1 0 0>, 2=<1 0 0>, 3=?, Index=$Particles.PartitionParticles.Partition, Num=::NewEnumerator1)
    Sprite Facing=&MergeHelper_VectorSwitch(0=<0 -1 0>, 1=<0.5 0 1>, 2=<-0.25 0 1>, 3=?, Index=$Particles.PartitionParticles.Partition, Num=::NewEnumerator1)
    WriteAlignment=true
    WriteFacing=true
   )
   &CameraOffset(Camera Offset Amount=&MergeHelper_FloatSwitch(0=0, 1=0, 2=5, 3=?, Index=$Particles.PartitionParticles.Partition, Num=::NewEnumerator1))
   /SetVariables(Particles.VisibilityTag=$Particles.PartitionParticles.Partition) {isSetParameters=true}
  ]
  particleUpdateScript [
   &ParticleState(Kill Particles When Lifetime Has Elapsed=false, Loop Particles Lifetime=true)
   &ScaleColor(
    Scale Alpha=&Multiply_Float(
     A=&FloatFromCurve(
      FloatCurve=DataInterface() {propertyValues {Curve {
         defaultValue=1
         keys [{interpMode=RCIM_Cubic, tangentMode=RCTM_User, time=0, value=0}, {interpMode=RCIM_Cubic, tangentMode=RCTM_User, time=1, value=1}]
        }}}
      CurveIndex=$Emitter.Age
      Scale Curve=&FN_FrontEndFloatModifier(Gameplay=0.25, Front End=0.15)
     )
     B=&FloatFromCurve(
      FloatCurve=DataInterface() {propertyValues {Curve {keys [
          <at 0 0.675 RCIM_Cubic>
          {arriveTangent=0.2812, interpMode=RCIM_Cubic, leaveTangent=0.2812, time=0.3, value=1}
          {arriveTangent=0.7821, interpMode=RCIM_Cubic, leaveTangent=0.7821, tangentMode=RCTM_User, time=0.8, value=0.9}
          <at 1 0.675 RCIM_Cubic>
         ]}}}
      Scale Curve=&MergeHelper_FloatSwitch(0=1, 1=1, 2=2, 3=?, Index=$Particles.PartitionParticles.Partition, Num=::NewEnumerator1)
     )
    )
    Scale RGB=&MakeVectorFromLinearColorRGB(Color=&VariantColorSelector(Variant Amount=1, Original Color=<rgba 1 0.02 0.02 0.25>, Variant 1 Color=<rgba 0 1 0.6462 0.25>, Variant 2 Color=<rgba 1 0.5659 0 0.25>))
    ScaleA=true
    ScaleRGB=true
   )
  ]
  rendererClasses [&NiagaraSpriteRendererProperties]
  renderers [
   {rendererClass=&NiagaraSpriteRendererProperties}
   {rendererClass=&NiagaraSpriteRendererProperties, rendererIndex=1}
   {rendererClass=&NiagaraSpriteRendererProperties, rendererIndex=2}
  ]
 }
}
summary {
 systemName=NS_Pickaxe_CopperMoggy_Idle
 emitters [
  {emitterName=Light, rendererClasses [&NiagaraLightRendererProperties]}
  {emitterName=Three_Way_Glow, rendererClasses [&NiagaraSpriteRendererProperties]}
 ]
 userVariables [
  User.bIsFrontEnd {type=NiagaraBool}
  User.bIsFrontEndPreview {type=NiagaraBool}
  User.Variant {type=NiagaraFloat, defaultValue=0, description=""}
 ]
}
