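# Niagara impact effect for the CopperMoggy pickaxe, with short bursts of claw marks, a lit flare, star flash, smoke, sparks and tech shapes.
# A shared system direction vector orients the claws and motion, while front-end modifiers adjust scale.
# A variant manager drives three palettes through color selectors in every emitter except smoke and tech shapes.
# Tech shapes instead compute a visibility tag from variant index plus a random bit, across six sprite renderers.
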NiagaraSystem /BRCosmetics/Weapons/FORT_Melee/Pickaxe_CopperMoggy_Swan/FX/NS_Pickaxe_CopperMoggy_Impact
dependencies {
 usedDataInterfaces [
  &NiagaraDataInterfaceCurve
  &NiagaraDataInterfaceVector2DCurve
  &NiagaraDataInterfaceVectorCurve
  &NiagaraDataInterfaceColorCurve
  &NiagaraDataInterfaceArrayFloat
  &NiagaraDataInterfaceMeshRendererInfo
  &NiagaraDataInterfaceArrayFloat3
 ]
 usedDynamicInputs [&MakeCustomFloatFromBool, /Game/Effects/Niagara/Modules/Utility/FN_DirectionVectorByPickaxeType]
 usedModules [
  &self:SystemSpawnScript.SystemScriptSource.SystemScriptGraph.NiagaraNodeAssignment_1.SetVariables_#1
  &SystemState
  &self:SystemSpawnScript.SystemScriptSource.SystemScriptGraph.NiagaraNodeAssignment_0.SetVariables_#2
  &FN_VariantManager
 ]
}
emitters {
 Claws {
  emitterUpdateScript [
   &EmitterState()
   &SpawnBurst_Instantaneous(Use Spawn Probability=false, UseLoopCountLimit=false)
  ]
  particleSpawnScript [
   &InitializeParticle(
    Lifetime=0.22
    Sprite Size=&Multiply_Vector2D_ByFloat(Float=&FN_FrontEndFloatModifier(Gameplay=1, Front End=$System.FrontendScale), Vector2D=<80 130>)
    Sprite Size Mode=::NewEnumerator4
    UsePositionOffset=false
    Write Material Random=true
   )
   &SpriteFacingAndAlignment(Sprite Alignment=$System.DirectionVector, WriteAlignment=true, WriteFacing=false)
  ]
  particleUpdateScript [
   &ParticleState()
   &ScaleSpriteSize(
    Scale Factor=&Vector2DFromCurve(
     Vector2Curve=DataInterface() {propertyValues {
       XCurve {keys [
         {arriveTangent=3.6667, interpMode=RCIM_Cubic, leaveTangent=3.6667, tangentMode=RCTM_User, time=0, value=0.25}
         {arriveTangent=-0.0049, interpMode=RCIM_Cubic, leaveTangent=-0.0049, tangentMode=RCTM_User, time=0.5, value=0.9}
         {arriveTangent=-0.7443, tangentMode=RCTM_User, time=1, value=0.3}
        ]}
       YCurve {keys [
         {arriveTangent=1.5117, interpMode=RCIM_Cubic, leaveTangent=1.5117, tangentMode=RCTM_User, time=0, value=0.3}
         {arriveTangent=-0.0046, interpMode=RCIM_Cubic, leaveTangent=-0.0046, tangentMode=RCTM_User, time=0.6, value=1}
         {arriveTangent=-1.9501, interpMode=RCIM_Cubic, leaveTangent=-1.9501, tangentMode=RCTM_User, time=1, value=0.4}
        ]}
      }}
     ScaleCurve=<1 1.5>
    )
    Scale Sprite Size Mode=::NewEnumerator2
   )
   &DynamicMaterialParameters(
    Index 0 Param 1=0.1
    Index 0 Param 2=&FloatFromCurve(
     FloatCurve=DataInterface() {propertyValues {Curve {keys [
         {arriveTangent=-1.5453, interpMode=RCIM_Cubic, leaveTangent=-1.5453, tangentMode=RCTM_User, time=0, value=0.3}
         {interpMode=RCIM_Cubic, tangentMode=RCTM_User, time=1, value=-0.4}
        ]}}}
     Scale Curve=0.5
    )
    Module.Index0Param0Write=true
    Module.Index0Param1Write=true
    Module.Index0Param2Write=true
    Module.Index0Param3Write=true
   )
   /SetVariables(Particles.Position=&Add_Vector(
     A=Linked() {linkedVariable {name=Particles.Initial.Position, type=NiagaraPosition}}
     B=&Multiply_VectorByFloat(
      Vector=$System.DirectionVector
      Float=&FloatFromCurve(
       FloatCurve=DataInterface() {propertyValues {Curve {keys [
           {arriveTangent=-2.9701, interpMode=RCIM_Cubic, leaveTangent=-2.9701, tangentMode=RCTM_User, time=0, value=-0.05}
           {arriveTangent=0.0501, interpMode=RCIM_Cubic, leaveTangent=0.0501, tangentMode=RCTM_User, time=0.1, value=-0.25}
           <at 1 0.6 RCIM_Cubic>
          ]}}}
       Scale Curve=30
      )
     )
    )) {isSetParameters=true}
   /SetVariables(
    Particles.SpriteSizeBright:Vector2f=/Niagara/DynamicInputs/Add/Add_Vector2(A:Vector2f=$Particles.SpriteSize, B:Vector2f=<-40 -30>)
    Particles.SpriteSizeDark:Vector2f=&Add_Vector2(A:Vector2f=$Particles.SpriteSize, B:Vector2f=<20 50>)
   ) {isSetParameters=true}
   &Color(Color=&VariantColorSelector(
     Variant Amount=1
     Original Color=&ColorFromCurve(ColorCurve=DataInterface() {propertyValues {
        AlphaCurve {keys [<at 0.5094 1 RCIM_Cubic>,<at 1 0 RCIM_Cubic>]}
        BlueCurve {keys [<at 0 0 RCIM_Cubic>,<at 0.5945 0 RCIM_Cubic>]}
        GreenCurve {keys [<at 0 0.05 RCIM_Cubic>,<at 0.5945 0.0257 RCIM_Cubic>]}
        RedCurve {keys [<at 0 3 RCIM_Cubic>,<at 0.5945 0.513 RCIM_Cubic>]}
       }})
     Variant 1 Color=&ColorFromCurve(ColorCurve=DataInterface() {propertyValues {
        AlphaCurve {keys [<at 0.1798 1 RCIM_Cubic>,<at 1 0 RCIM_Cubic>]}
        BlueCurve {keys [<at 0 0.3049 RCIM_Cubic>,<at 0.5945 0.5076 RCIM_Cubic>]}
        GreenCurve {keys [<at 0 0.5 RCIM_Cubic>,<at 0.5945 0.513 RCIM_Cubic>]}
        RedCurve {keys [<at 0 0 RCIM_Cubic>,<at 0.5945 0 RCIM_Cubic>]}
       }})
     Variant 2 Color=&ColorFromCurve(ColorCurve=DataInterface() {propertyValues {
        AlphaCurve {keys [<at 0.1798 1 RCIM_Cubic>,<at 1 0 RCIM_Cubic>]}
        BlueCurve {keys [<at 0 0 RCIM_Cubic>,<at 0.5945 0 RCIM_Cubic>]}
        GreenCurve {keys [<at 0 2.2193 RCIM_Cubic>,<at 0.5945 0.2747 RCIM_Cubic>]}
        RedCurve {keys [<at 0 3 RCIM_Cubic>,<at 0.5945 0.6979 RCIM_Cubic>]}
       }})
    ))
   /SetVariables(
    Particles.Bright Color:LinearColor=&VariantColorSelector(
     Variant Amount=1
     Original Color=&ColorFromCurve(ColorCurve=DataInterface() {propertyValues {
        AlphaCurve {keys [<at 0.3289 5 RCIM_Cubic>,<at 1 0 RCIM_Cubic>]}
        BlueCurve {keys [<at 0.0738 1.2 RCIM_Cubic>,<at 0.8335 0 RCIM_Cubic>]}
        GreenCurve {keys [<at 0.0738 1.56 RCIM_Cubic>,<at 0.8335 0 RCIM_Cubic>]}
        RedCurve {keys [<at 0.0738 12 RCIM_Cubic>,<at 0.8335 1 RCIM_Cubic>]}
       }})
     Variant 1 Color=&ColorFromCurve(ColorCurve=DataInterface() {propertyValues {
        AlphaCurve {keys [<at 0.3289 5 RCIM_Cubic>,<at 1 0 RCIM_Cubic>]}
        BlueCurve {keys [<at 0.0738 6.1119 RCIM_Cubic>,<at 0.6795 1 RCIM_Cubic>]}
        GreenCurve {keys [<at 0.0738 12 RCIM_Cubic>,<at 0.6795 0.7662 RCIM_Cubic>]}
        RedCurve {keys [<at 0.0738 0 RCIM_Cubic>,<at 0.6795 0 RCIM_Cubic>]}
       }})
     Variant 2 Color=&ColorFromCurve(ColorCurve=DataInterface() {propertyValues {
        AlphaCurve {keys [<at 0.3289 5 RCIM_Cubic>,<at 1 0 RCIM_Cubic>]}
        BlueCurve {keys [<at 0.0738 0.6 RCIM_Cubic>,<at 0.9087 0 RCIM_Cubic>]}
        GreenCurve {keys [<at 0.0738 6.3 RCIM_Cubic>,<at 0.9087 0.4174 RCIM_Cubic>]}
        RedCurve {keys [<at 0.0738 12 RCIM_Cubic>,<at 0.9087 1 RCIM_Cubic>]}
       }})
    )
    Particles.Dark Color:LinearColor=&VariantColorSelector(
     Variant Amount=1
     Original Color=&ColorFromCurve(ColorCurve=DataInterface() {propertyValues {
        AlphaCurve {keys [<at 0.2051 3 RCIM_Cubic>,<at 1 0 RCIM_Cubic>]}
        BlueCurve {keys [<at 0.1335 0 RCIM_Cubic>]}
        GreenCurve {keys [<at 0.1335 0 RCIM_Cubic>]}
        RedCurve {keys [<at 0.1335 0.1 RCIM_Cubic>]}
       }})
     Variant 1 Color=&ColorFromCurve(ColorCurve=DataInterface() {propertyValues {
        AlphaCurve {keys [<at 0.2051 4 RCIM_Cubic>,<at 1 0 RCIM_Cubic>]}
        BlueCurve {keys [<at 0.117 0.0882 RCIM_Cubic>]}
        GreenCurve {keys [<at 0.117 0.1 RCIM_Cubic>]}
        RedCurve {keys [<at 0.117 0 RCIM_Cubic>]}
       }})
     Variant 2 Color=&ColorFromCurve(ColorCurve=DataInterface() {propertyValues {
        AlphaCurve {keys [<at 0.2051 4 RCIM_Cubic>,<at 1 0 RCIM_Cubic>]}
        BlueCurve {keys [<at 0.1335 0 RCIM_Cubic>]}
        GreenCurve {keys [<at 0.1335 0.089 RCIM_Cubic>]}
        RedCurve {keys [<at 0.1335 0.3 RCIM_Cubic>]}
       }})
    )
   ) {isSetParameters=true}
  ]
  rendererClasses [&NiagaraSpriteRendererProperties]
  renderers [
   {rendererClass=&NiagaraSpriteRendererProperties}
   {rendererClass=&NiagaraSpriteRendererProperties, rendererIndex=1}
   {rendererClass=&NiagaraSpriteRendererProperties, rendererIndex=2}
  ]
 }
 Flash_Flare {
  emitterUpdateScript [
   &EmitterState()
   &SpawnBurst_Instantaneous(Use Spawn Probability=false, UseLoopCountLimit=false)
  ]
  particleSpawnScript [
   &InitializeParticle(
    Color=&VariantColorSelector(Variant Amount=1, Original Color=<rgba 6 0.1 0 1>, Variant 1 Color=<rgba 0 3 1.9378 1>, Variant 2 Color=<rgba 4 1.4667 0 1>)
    Lifetime=0.15
    Sprite Size=&Multiply_Vector2D_ByFloat(Float=&FN_FrontEndFloatModifier(Gameplay=1, Front End=$System.FrontendScale), Vector2D=<100 200>)
    Sprite Size Mode=::NewEnumerator4
    UsePositionOffset=false
    Write Material Random=true
   )
   &DynamicMaterialParameters(
    Index 0 Param 1=0.2
    Index 0 Param 2=1
    Module.Param0WriteEnabled=true
    Module.Param1WriteEnabled=true
    Module.Param2WriteEnabled=true
    Module.Param3WriteEnabled=true
   )
   /SetVariables(Particles.Velocity=$System.DirectionVector) {isSetParameters=true}
  ]
  particleUpdateScript [
   &ParticleState()
   &ScaleColor(
    Scale Alpha=&FloatFromCurve(
     FloatCurve=DataInterface() {propertyValues {Curve {
        defaultValue=1
        keys [
         {arriveTangent=25.4958, interpMode=RCIM_Cubic, leaveTangent=25.4958, tangentMode=RCTM_User, time=0, value=0}
         {arriveTangent=-2.5249, interpMode=RCIM_Cubic, leaveTangent=-2.5249, tangentMode=RCTM_User, time=0.1, value=1}
         {arriveTangent=-0.0405, interpMode=RCIM_Cubic, leaveTangent=-0.0405, tangentMode=RCTM_User, time=1, value=0}
        ]
       }}}
     Scale Curve=10
    )
    Scale RGB=&VectorFromFloat(Value=1.5)
    ScaleA=true
    ScaleRGB=true
   )
   &FN_Light_v2(Light Exponent=5, Light Radius=50, LightEnabled=true, Scale Color=1)
  ]
  rendererClasses [&NiagaraSpriteRendererProperties, &NiagaraLightRendererProperties]
  renderers [
   {rendererClass=&NiagaraSpriteRendererProperties}
   {rendererClass=&NiagaraLightRendererProperties, rendererIndex=1}
  ]
 }
 Flash_Star {
  emitterUpdateScript [
   &EmitterState()
   &SpawnBurst_Instantaneous(Spawn Time=0.025, Use Spawn Probability=false, UseLoopCountLimit=false)
  ]
  particleSpawnScript [
   &InitializeParticle(
    Color=&VariantColorSelector(Variant Amount=1, Original Color=<rgba 1 0.0333 0 0.75>, Variant 1 Color=<rgba 0 1 0.5067 0.7>, Variant 2 Color=<rgba 1 0.525 0.05 0.75>)
    Lifetime=0.18
    Sprite Rotation Mode=::NewEnumerator2
    Sprite Size=<40 120>
    Sprite Size Mode=::NewEnumerator4
    UsePositionOffset=false
    Write Material Random=true
   )
   &FN_PerParticleLOD(Enable SpriteSize Distance Scaling=true, Sprite Size Scale Far=<2.5 2.5>)
   &CameraOffset(Camera Offset Amount=15)
   /SetVariables(Particles.RandomFloat=&UniformRangedFloat(Minimum=0, Maximum=1)) {isSetParameters=true}
   /SetVariables(Particles.SubImageIndex=&RandomRangeFloat(Minimum=0, Maximum=3)) {isSetParameters=true}
   /SetVariables(Particles.SubImageIndexAlt:NiagaraFloat=&ModuloFloat(A=&Add_Float(A=$Particles.SubImageIndex, B=1), B=4)) {isSetParameters=true}
   /SetVariables(Particles.Velocity=$System.DirectionVector) {isSetParameters=true}
  ]
  particleUpdateScript [
   &ParticleState()
   &ScaleColor(
    Scale Alpha=&Multiply_Float(
     A=&FloatFromCurve(FloatCurve=DataInterface() {propertyValues {Curve {
         defaultValue=1
         keys [
          {arriveTangent=-0.4174, interpMode=RCIM_Cubic, leaveTangent=-0.4174, tangentMode=RCTM_User, time=0, value=1}
          {interpMode=RCIM_Cubic, tangentMode=RCTM_User, time=1, value=0}
         ]
        }}})
     B=&FNTimeOfDayScalar(Night Scale=1, Morning Scale=1, Evening Scale=1)
    )
    Scale RGB=&VectorFromFloat(Value=&FloatFromCurve(
      FloatCurve=DataInterface() {propertyValues {Curve {keys [{arriveTangent=-0.0343, interpMode=RCIM_Cubic, leaveTangent=-0.0343, tangentMode=RCTM_User, time=0, value=1}, <at 1 0.1>]}}}
      Scale Curve=4
     ))
    ScaleA=true
    ScaleRGB=true
   )
   &ScaleSpriteSize(
    Uniform Curve Scale=&Multiply_Float(A=1.5, B=&FN_FrontEndFloatModifier(Gameplay=1, Front End=$System.FrontendScale))
    Uniform Curve Sprite Scale=DataInterface() {propertyValues {Curve {keys [{arriveTangent=1.454, interpMode=RCIM_Cubic, leaveTangent=1.454, tangentMode=RCTM_User, time=0, value=0.4}, <at 1 1 RCIM_Cubic>]}}}
   ) {module=ScaleSpriteSize001}
   /SetVariables(Particles.SpriteSizeAlt:Vector2f=&Multiply_Vector2D(A=$Particles.SpriteSize, B=<1.3 1.3>)) {isSetParameters=true}
   /SetVariables(Particles.ColorAlt:LinearColor=&VariantColorSelector(
     Variant Amount=1
     Original Color=&ColorFromCurve(ColorCurve=DataInterface() {propertyValues {
        AlphaCurve {keys [<at 0 1.1 RCIM_Cubic>,<at 1 0 RCIM_Cubic>]}
        BlueCurve {keys [<at 0 0 RCIM_Cubic>]}
        GreenCurve {keys [<at 0 0.0083 RCIM_Cubic>]}
        RedCurve {keys [<at 0 0.1 RCIM_Cubic>]}
       }})
     Variant 1 Color=&ColorFromCurve(ColorCurve=DataInterface() {propertyValues {
        AlphaCurve {keys [<at 0 1.2 RCIM_Cubic>,<at 1 0 RCIM_Cubic>]}
        BlueCurve {keys [<at 0 0.0313 RCIM_Cubic>]}
        GreenCurve {keys [<at 0 0.035 RCIM_Cubic>]}
        RedCurve {keys [<at 0 0 RCIM_Cubic>]}
       }})
     Variant 2 Color=&ColorFromCurve(ColorCurve=DataInterface() {propertyValues {
        AlphaCurve {keys [<at 0 0.8 RCIM_Cubic>,<at 1 0 RCIM_Cubic>]}
        BlueCurve {keys [<at 0 0 RCIM_Cubic>]}
        GreenCurve {keys [<at 0 0.2665 RCIM_Cubic>]}
        RedCurve {keys [<at 0 0.8 RCIM_Cubic>]}
       }})
    )) {isSetParameters=true}
  ]
  rendererClasses [&NiagaraSpriteRendererProperties]
  renderers [
   {rendererClass=&NiagaraSpriteRendererProperties}
   {rendererClass=&NiagaraSpriteRendererProperties, rendererIndex=1}
  ]
 }
 SmokePuff {
  emitterUpdateScript [
   &EmitterState()
   &SpawnBurst_Instantaneous(
    Spawn Count=&RandomRangeInt(Minimum=2, Maximum=3)
    Spawn Time=0.02
    Use Spawn Probability=false
    UseLoopCountLimit=false
   )
  ]
  particleSpawnScript [
   &InitializeParticle(
    Color Maximum=<rgba 0.76 1 1 0.2>
    Color Minimum=<rgba 0.8 0.917 1 0.15>
    Color Mode=::NewEnumerator2
    Lifetime Max=0.5
    Lifetime Min=0.2
    Lifetime Mode=::NewEnumerator1
    Sprite Size Mode=::NewEnumerator3
    Uniform Sprite Size=&Multiply_Float(
     A=&FN_FrontEndFloatModifier(Gameplay=1, Front End=$System.FrontendScale)
     B=&RandomRangeFloat(Minimum=100, Maximum=140)
    )
    UsePositionOffset=false
    Write Material Random=true
   )
   &CameraOffset(Camera Offset Amount=&RandomRangeFloat(Minimum=40, Maximum=50, Evaluation Type=::NewEnumerator1))
   &ShapeLocation(
    Cylinder Height=20
    Cylinder Radius=5
    MaskSpawnGroup=false
    Override Seed=false
    Rotation Axis=$System.DirectionVector
    Rotation Mode=::NewEnumerator1
    Shape Primitive=::NewEnumerator1
    Surface Only=false
    UseOwnerScale=true
   ) {module=ShapeLocation001}
   &AddVelocity(
    Cone Angle=60
    Cone Axis=$System.DirectionVector
    Module.UseConeFalloff=false
    Module.UseVelDistribution=false
    Override Seed=false
    Velocity Mode=::NewEnumerator2
    Velocity Speed=&RandomRangeFloat(Minimum=100, Maximum=200)
   ) {module=AddVelocity001}
  ]
  particleUpdateScript [
   &ParticleState()
   &ScaleSpriteSize(Non-Uniform Curve Scale=<1 2>, Scale Sprite Size Mode=::NewEnumerator3)
   &ScaleColor(
    Scale Alpha=&FloatFromCurve(
     FloatCurve=DataInterface() {propertyValues {Curve {keys [
         {interpMode=RCIM_Cubic, tangentMode=RCTM_User, time=0, value=0.4}
         {interpMode=RCIM_Cubic, tangentMode=RCTM_User, time=0.3, value=1}
         {tangentMode=RCTM_User, time=1, value=0}
        ]}}}
     Scale Curve=0.3
    )
    Scale RGB=&VectorFromCurve(VectorCurve=DataInterface() {propertyValues {XCurve {keys [<at 0 1>,<at 1 1>]}, YCurve {keys [<at 0 1>,<at 1 1>]}, ZCurve {keys [<at 0 1>,<at 1 1>]}}})
    ScaleA=true
    ScaleRGB=true
   )
   &DynamicMaterialParameters(
    Index 0 Param 1=&Add_Float(
     A=&RandomRangeFloat(Minimum=0, Maximum=1)
     B=&FloatFromCurve(
      FloatCurve=DataInterface() {propertyValues {Curve {keys [
          <at 0 0 RCIM_Cubic>
          {arriveTangent=1.5947, interpMode=RCIM_Cubic, leaveTangent=1.5947, tangentMode=RCTM_User, time=0.2557, value=0.5257}
          <at 1 1 RCIM_Cubic>
         ]}}}
      CurveIndex=$Particles.Age
      Scale Curve=0.25
     )
    )
    Module.Param0WriteEnabled=true
    Module.Param1WriteEnabled=true
    Module.Param2WriteEnabled=true
    Module.Param3WriteEnabled=true
   ) {module=DynamicMaterialParameters001}
   &SolveForcesAndVelocity(Clamp Velocity=false, Limit Acceleration=false)
  ]
  rendererClasses [&NiagaraSpriteRendererProperties]
  renderers [
   {rendererClass=&NiagaraSpriteRendererProperties}
   {rendererClass=&NiagaraSpriteRendererProperties, rendererIndex=1}
  ]
 }
 Sparks {
  emitterUpdateScript [
   &EmitterState()
   &SpawnBurst_Instantaneous(Spawn Count=&UniformRangedInt(Minimum=16, Maximum=22), Spawn Time=0.02, Use Spawn Probability=false)
  ]
  particleSpawnScript [
   &FN_SetSpawnGroup(Number of Spawn Groups=5)
   &InitializeParticle(
    Lifetime Max=0.35
    Lifetime Min=0.25
    Lifetime Mode=::NewEnumerator1
    Sprite Size Max=<10 12>
    Sprite Size Min=<8 10>
    Sprite Size Mode=::NewEnumerator2
    UsePositionOffset=false
    Write Material Random=true
   )
   &CameraOffset(Camera Offset Amount=&FN_FrontEndFloatModifier(Gameplay=10, Front End=$System.FrontendCameraOffset))
   &ShapeLocation(
    Cylinder Height=&Multiply_Float(A=75, B=&FN_FrontEndFloatModifier(Gameplay=1, Front End=$System.FrontendScale))
    Cylinder Height Midpoint:NiagaraFloat=&MaskFloatBySpawnGroup(Masked Float=0.8, Passthrough Float=0.4)
    Cylinder Radius=&FN_FrontEndFloatModifier(Gameplay=1, Front End=$System.FrontendScale)
    MaskSpawnGroup=false
    Override Seed=false
    Rotation Axis=$System.DirectionVector
    Rotation Mode=::NewEnumerator1
    Shape Primitive=::NewEnumerator1
    Surface Only=false
    UseOwnerScale=true
   )
   &AddVelocity(
    Cone Angle=70
    Cone Axis=$System.DirectionVector
    Module.UseConeFalloff=false
    Module.UseVelDistribution=true
    Override Seed=false
    Velocity Mode=::NewEnumerator2
    Velocity Speed=&Multiply_Float(
     A=&Multiply_Float(
      A=&RandomRangeFloat(Minimum=300, Maximum=600)
      B=&FN_FrontEndFloatModifier(Gameplay=1, Front End=$System.FrontendScale)
     )
     B=&MaskFloatBySpawnGroup(Masked Float=-0.6, Passthrough Float=1)
    )
   )
  ]
  particleUpdateScript [
   &ParticleState()
   &ScaleColor(
    Scale Alpha=&FloatFromCurve(FloatCurve=DataInterface() {propertyValues {Curve {
        defaultValue=1
        keys [
         {arriveTangent=-1.7554, interpMode=RCIM_Cubic, leaveTangent=-1.7554, tangentMode=RCTM_User, time=0, value=1}
         {interpMode=RCIM_Cubic, tangentMode=RCTM_User, time=1, value=0}
        ]
       }}})
    Scale RGB=&MakeVectorFromLinearColorRGB(Color=&VariantColorSelector(
      Variant Amount=1
      Original Color=&ColorFromCurve(ColorCurve=DataInterface() {propertyValues {
         AlphaCurve {keys [<at 0 1>]}
         BlueCurve {keys [<at 0 0.1>,<at 1 0>]}
         GreenCurve {keys [<at 0 0.1633>,<at 1 0.0333>]}
         RedCurve {keys [<at 0 2>,<at 1 1>]}
        }})
      Variant 1 Color=&ColorFromCurve(ColorCurve=DataInterface() {propertyValues {
         AlphaCurve {keys [<at 0 1>]}
         BlueCurve {keys [<at 0 1.2261>,<at 0.5109 1>]}
         GreenCurve {keys [<at 0 2>,<at 0.5109 0.6012>]}
         RedCurve {keys [<at 0 0.3031>,<at 0.5109 0>]}
        }})
      Variant 2 Color=&ColorFromCurve(ColorCurve=DataInterface() {propertyValues {
         AlphaCurve {keys [<at 0 1>]}
         BlueCurve {keys [<at 0 0.4>,<at 0.6722 0>]}
         GreenCurve {keys [<at 0 1.2>,<at 0.6722 0.3333>]}
         RedCurve {keys [<at 0 2>,<at 0.6722 1>]}
        }})
     ))
    ScaleA=true
    ScaleRGB=true
   )
   &ScaleSpriteSize(Scale Factor=&Vector2DFromFloat(Value=&FloatFromCurve(
      FloatCurve=DataInterface() {propertyValues {Curve {keys [
          {arriveTangent=10.1867, interpMode=RCIM_Cubic, leaveTangent=10.1867, tangentMode=RCTM_User, time=0, value=0}
          {interpMode=RCIM_Cubic, leaveTangent=-0.0263, tangentMode=RCTM_Break, time=0.2, value=1.2}
          {arriveTangent=-2.7631, interpMode=RCIM_Cubic, leaveTangent=-2.7631, tangentMode=RCTM_User, time=1, value=0}
         ]}}}
      Scale Curve=&Multiply_Float(A=0.4, B=$System.FrontendScale)
     ))) {module=ScaleSpriteSize001}
   &CurlNoiseForce(Noise Strength=200, Pan Noise=true, Pan Noise Field=<0 0 0.5>, Random Seed=10)
   &Drag(Drag=8, Use Linear Drag=true, Use Rotational Drag=false)
   &SolveForcesAndVelocity(Clamp Velocity=false, Limit Acceleration=false)
   &ScaleSpriteSizeBySpeed(Max Scale Factor=<0.5 4>, Min Scale Factor=<1 2>, Velocity Threshold=100)
   /SetVariables(Particles.Glow_Color:LinearColor=&MakeLinearColorFromVectorAndFloat(
     "Vector (RGB)"=&MakeVectorFromLinearColorRGB(Color=&VariantColorSelector(
       Variant Amount=1
       Original Color=&ColorFromCurve(ColorCurve=DataInterface() {propertyValues {
          AlphaCurve {keys [<at 0 1>,<at 0.8992 0>]}
          BlueCurve {keys [<at 0 0.04>,<at 1 0>]}
          GreenCurve {keys [<at 0 0.078>,<at 1 0.0167>]}
          RedCurve {keys [<at 0 0.8>,<at 1 0.2>]}
         }})
       Variant 1 Color=&ColorFromCurve(ColorCurve=DataInterface() {propertyValues {
          AlphaCurve {keys [<at 0 1>,<at 0.8992 0>]}
          BlueCurve {keys [<at 0 0.2667>,<at 1 0.1333>]}
          GreenCurve {keys [<at 0 0.8>,<at 1 0.2>]}
          RedCurve {keys [<at 0 0>,<at 1 0>]}
         }})
       Variant 2 Color=&ColorFromCurve(ColorCurve=DataInterface() {propertyValues {
          AlphaCurve {keys [<at 0 1>,<at 0.8992 0>]}
          BlueCurve {keys [<at 0 0>,<at 1 0>]}
          GreenCurve {keys [<at 0 0.2667>,<at 1 0.05>]}
          RedCurve {keys [<at 0 0.8>,<at 1 0.2>]}
         }})
      ))
     "Float (Alpha)"=0.5
    )) {isSetParameters=true}
   /SetVariables(Particles.Glow_Size:Vector2f=&Multiply_Vector2D(A=$Particles.SpriteSize, B=<4 2>)) {isSetParameters=true}
  ]
  rendererClasses [&NiagaraSpriteRendererProperties]
  renderers [
   {rendererClass=&NiagaraSpriteRendererProperties}
   {rendererClass=&NiagaraSpriteRendererProperties, rendererIndex=1}
  ]
 }
 Tech_Shapes {
  emitterUpdateScript [
   &EmitterState()
   &SpawnBurst_Instantaneous(Spawn Count=&UniformRangedInt(Minimum=4, Maximum=6), Spawn Time=0.025, Use Spawn Probability=false)
  ]
  particleSpawnScript [
   &InitializeParticle(
    Lifetime Max=0.5
    Lifetime Min=0.35
    Lifetime Mode=::NewEnumerator1
    Sprite Rotation Mode=::NewEnumerator2
    Sprite Size Mode=::NewEnumerator1
    Uniform Sprite Size Max=9
    Uniform Sprite Size Min=7
    UsePositionOffset=false
    Write Material Random=true
   )
   &ShapeLocation(
    Cylinder Height=&Multiply_Float(A=50, B=&FN_FrontEndFloatModifier(Gameplay=1, Front End=$System.FrontendScale))
    Cylinder Height Midpoint=0.25
    Cylinder Radius=&Multiply_Float(A=15, B=&FN_FrontEndFloatModifier(Gameplay=1, Front End=$System.FrontendScale))
    MaskSpawnGroup=false
    Override Seed=false
    Rotation Axis=$System.DirectionVector
    Rotation Mode=::NewEnumerator1
    Shape Primitive=::NewEnumerator1
    Surface Only=false
    UseOwnerScale=true
   )
   &AddVelocity(
    Cone Angle=40
    Cone Axis=&Add_Vector(A=$System.DirectionVector, B=<0.2 0 0>)
    Module.UseConeFalloff=false
    Module.UseVelDistribution=false
    Override Seed=false
    Velocity Mode=::NewEnumerator2
    Velocity Speed=&Multiply_Float(
     A=&RandomRangeFloat(Minimum=200, Maximum=600)
     B=&FN_FrontEndFloatModifier(Gameplay=1, Front End=$System.FrontendScale)
    )
   )
   /SetVariables(Particles.VisibilityTag=&RandomRangeInt(Minimum=0, Maximum=1)) {isSetParameters=true}
   /SetVariables(Particles.VisibilityTag=&Add_Integer(
     A=$Particles.VisibilityTag
     B=&Multiply_Int(A=&MakeIntFromFloat(Float=$System.VariantIndex), B=2)
    )) {isSetParameters=true}
   &CameraOffset(Camera Offset Amount=5)
  ]
  particleUpdateScript [
   &ParticleState()
   &SpriteRotationRate(Rotation Rate=&SelectFloatFromArray(
     Array Sampling Mode=::NewEnumerator0
     Direct Array Index=?
     Float Selection Array=DataInterface() {propertyValues {FloatData [-100,150,200]}}
     Random Seed=0
    ))
   &CurlNoiseForce(Noise Strength=200, Pan Noise=false) {module=CurlNoiseForce001}
   &Drag(
    Drag=&FloatFromCurve(
     FloatCurve=DataInterface() {propertyValues {Curve {keys [
         {arriveTangent=0.0075, interpMode=RCIM_Cubic, leaveTangent=0.0075, tangentMode=RCTM_User, time=0, value=0.4}
         {arriveTangent=1.183, interpMode=RCIM_Cubic, leaveTangent=1.183, tangentMode=RCTM_User, time=1, value=1}
        ]}}}
     Scale Curve=&RandomRangeFloat(Minimum=12, Maximum=15)
    )
    Use Linear Drag=true
    Use Rotational Drag=false
   )
   &SolveForcesAndVelocity(Clamp Velocity=false, Limit Acceleration=false)
   &ScaleSpriteSize(Scale Factor=&Vector2DFromFloat(Value=&FloatFromCurve(
      FloatCurve=DataInterface() {propertyValues {Curve {keys [
          {arriveTangent=1.5239, interpMode=RCIM_Cubic, leaveTangent=1.5239, tangentMode=RCTM_User, time=0, value=0.3}
          {arriveTangent=0.0139, interpMode=RCIM_Cubic, leaveTangent=0.0139, tangentMode=RCTM_User, time=1, value=1}
         ]}}}
      Scale Curve=&Multiply_Float(A=1.2, B=&FN_FrontEndFloatModifier(Gameplay=1, Front End=$System.FrontendScale))
     )))
   &ScaleColor(
    Linear Color Curve=DataInterface() {propertyValues {
      AlphaCurve {keys [<at 0.182 1>,<at 1 0>]}
      BlueCurve {keys [<at 0.3 1.2>,<at 0.9435 0.2>]}
      GreenCurve {keys [<at 0.3 1.2>,<at 0.9435 0.2>]}
      RedCurve {keys [<at 0.3 1.2>,<at 0.9435 0.2>]}
     }}
    Scale Mode=::NewEnumerator2
   )
  ]
  rendererClasses [&NiagaraSpriteRendererProperties]
  renderers [
   {rendererClass=&NiagaraSpriteRendererProperties}
   {rendererClass=&NiagaraSpriteRendererProperties, rendererIndex=1}
   {rendererClass=&NiagaraSpriteRendererProperties, rendererIndex=2}
   {rendererClass=&NiagaraSpriteRendererProperties, rendererIndex=3}
   {rendererClass=&NiagaraSpriteRendererProperties, rendererIndex=4}
   {rendererClass=&NiagaraSpriteRendererProperties, rendererIndex=5}
  ]
 }
}
summary {
 systemName=NS_Pickaxe_CopperMoggy_Impact
 emitters [
  {emitterName=SmokePuff, rendererClasses [&NiagaraSpriteRendererProperties]}
  {emitterName=Flash_Flare, rendererClasses [&NiagaraSpriteRendererProperties, &NiagaraLightRendererProperties]}
  {emitterName=Flash_Star, rendererClasses [&NiagaraSpriteRendererProperties]}
  {emitterName=Claws, rendererClasses [&NiagaraSpriteRendererProperties]}
  {emitterName=Tech_Shapes, rendererClasses [&NiagaraSpriteRendererProperties]}
  {emitterName=Sparks, rendererClasses [&NiagaraSpriteRendererProperties]}
 ]
 userVariables [
  User.bIsFrontEnd {type=NiagaraBool}
  User.bIsFrontEndPreview {type=NiagaraBool}
  User.SwingSectionIndex {type=NiagaraInt32}
  User.Variant {type=NiagaraFloat, defaultValue=0, description=""}
 ]
}
