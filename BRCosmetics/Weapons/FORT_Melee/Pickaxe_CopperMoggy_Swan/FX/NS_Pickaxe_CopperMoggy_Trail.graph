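# Niagara swing trail for the CopperMoggy pickaxe, drawing animation-trail ribbons in three variant styles.
# Fill_Trail is a short-lived ribbon with variant-selected color and lifetime, fading with the system's swing fade.
# Tech_Trail adds time-of-day tint, width tapering and per-swing-section curves driving brightness and a material parameter.
# It writes Swan, Prime and Flame colors with variant-selected alpha alongside three ribbon renderers, leaving one color visible per variant.
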NiagaraSystem /BRCosmetics/Weapons/FORT_Melee/Pickaxe_CopperMoggy_Swan/FX/NS_Pickaxe_CopperMoggy_Trail
dependencies {
 usedDataInterfaces [&NiagaraDataInterfaceCurve, &NiagaraDataInterfaceSocketReader]
 usedDynamicInputs [&FNTimeOfDayColor]
 usedModules [
  &self:SystemSpawnScript.SystemScriptSource.SystemScriptGraph.NiagaraNodeAssignment_0.SetVariables_#1
  &SystemState
  &self:SystemSpawnScript.SystemScriptSource.SystemScriptGraph.NiagaraNodeAssignment_1.SetVariables_#2
  &self:SystemSpawnScript.SystemScriptSource.SystemScriptGraph.NiagaraNodeAssignment_2.SetVariables_#3
  &FN_VariantManager
 ]
}
emitters {
 Fill_Trail {
  emitterUpdateScript [
   &EmitterState()
   &FN_SpawnAnimTrail()
  ]
  particleSpawnScript [
   &FN_InitializeAnimTrailBoneIndecies()
   &FN_InitializeAnimTrail(
    AnimTrailWidth=0.95
    Color=&VariantColorSelector(Variant Amount=1, Original Color=<rgba 0 0 0 0.75>, Variant 1 Color=<rgba 0 1 0.7009 1>, Variant 2 Color=<rgba 1 0.4075 0 0.875>)
    Lifetime=/BRCosmetics/Weapons/FORT_Melee/Pickaxe_GateHound/FX/VariantManager/VariantFloatSelector(
     Original Float:NiagaraFloat=0.15
     Variant 1 Float:NiagaraFloat=0.15
     Variant 2 Float:NiagaraFloat=0.12
     Variant 3 Float:NiagaraFloat=?
     Variant Amount:NiagaraInt32=1
    )
   )
  ]
  particleUpdateScript [
   &FN_UpdateAnimTrail()
   &ScaleColor(
    Scale Alpha=&FloatFromCurve(
     FloatCurve=DataInterface() {propertyValues {Curve {
        defaultValue=1
        keys [{interpMode=RCIM_Cubic, tangentMode=RCTM_User, time=0, value=1}, {interpMode=RCIM_Cubic, tangentMode=RCTM_User, time=0.5, value=0}]
       }}}
     Scale Curve=$System.SwingGeneralFade
    )
    ScaleA=true
    ScaleRGB=true
   )
  ]
  rendererClasses [&NiagaraRibbonRendererProperties]
  renderers [{rendererClass=&NiagaraRibbonRendererProperties}]
 }
 Tech_Trail {
  emitterUpdateScript [
   &EmitterState()
   &FN_SpawnAnimTrail()
  ]
  particleSpawnScript [
   &FN_InitializeAnimTrailBoneIndecies()
   &FN_InitializeAnimTrail(Lifetime=0.25)
  ]
  particleUpdateScript [
   &FN_UpdateAnimTrail()
   &ScaleColor(
    Scale Alpha=&FloatFromCurve(
     FloatCurve=DataInterface() {propertyValues {Curve {
        defaultValue=1
        keys [
         {interpMode=RCIM_Cubic, tangentMode=RCTM_User, time=0, value=1}
         {arriveTangent=-2.6541, interpMode=RCIM_Cubic, leaveTangent=-2.6541, tangentMode=RCTM_User, time=1, value=0}
        ]
       }}}
     Scale Curve=$System.SwingGeneralFade
    )
    Scale RGB=&Multiply_VectorByFloat(
     Vector=&MakeVectorFromLinearColorRGB(Color=$System.TODColor)
     Float=&MergeHelper_FloatSwitch(
      0=&FloatFromCurve(
       FloatCurve=DataInterface() {propertyValues {Curve {
          defaultValue=1
          keys [{interpMode=RCIM_Cubic, tangentMode=RCTM_User, time=0.1, value=1}, {interpMode=RCIM_Cubic, tangentMode=RCTM_User, time=0.4, value=0.2}]
         }}}
       CurveIndex=$Emitter.Age
       Scale Curve=10
      )
      1=&FloatFromCurve(
       FloatCurve=DataInterface() {propertyValues {Curve {
          defaultValue=1
          keys [{interpMode=RCIM_Cubic, tangentMode=RCTM_User, time=0.1, value=1}, {interpMode=RCIM_Cubic, tangentMode=RCTM_User, time=0.3, value=0.2}]
         }}}
       CurveIndex=$Emitter.Age
       Scale Curve=10
      )
      2=&FloatFromCurve(
       FloatCurve=DataInterface() {propertyValues {Curve {
          defaultValue=1
          keys [{interpMode=RCIM_Cubic, tangentMode=RCTM_User, time=0.1, value=1}, {interpMode=RCIM_Cubic, tangentMode=RCTM_User, time=0.4, value=0.2}]
         }}}
       CurveIndex=$Emitter.Age
       Scale Curve=10
      )
      3=&FloatFromCurve(
       FloatCurve=DataInterface() {propertyValues {Curve {
          defaultValue=1
          keys [{interpMode=RCIM_Cubic, tangentMode=RCTM_User, time=0.1, value=1}, {interpMode=RCIM_Cubic, tangentMode=RCTM_User, time=0.3, value=0.2}]
         }}}
       CurveIndex=$Emitter.Age
       Scale Curve=10
      )
      Index=$User.SwingSectionIndex
      Num=::NewEnumerator2
     )
    )
    ScaleA=true
    ScaleRGB=true
   )
   &ScaleRibbonWidth(Ribbon Width Scale=&FloatFromCurve(FloatCurve=DataInterface() {propertyValues {Curve {
        defaultValue=1
        keys [{interpMode=RCIM_Cubic, tangentMode=RCTM_User, time=0, value=1}, {interpMode=RCIM_Cubic, tangentMode=RCTM_User, time=1, value=0.4}]
       }}}))
   &DynamicMaterialParameters(
    Index 0 Param 2=&MergeHelper_FloatSwitch(
     0=&FloatFromCurve(
      FloatCurve=DataInterface() {propertyValues {Curve {keys [
          {arriveTangent=3.2789, interpMode=RCIM_Cubic, time=0.3, value=0.1}
          {arriveTangent=0.0568, interpMode=RCIM_Cubic, leaveTangent=0.0568, tangentMode=RCTM_User, time=0.5, value=1}
         ]}}}
      CurveIndex=$Emitter.Age
     )
     1=&FloatFromCurve(
      FloatCurve=DataInterface() {propertyValues {Curve {keys [
          {arriveTangent=3.2789, interpMode=RCIM_Cubic, time=0.2, value=0.1}
          {arriveTangent=0.0568, interpMode=RCIM_Cubic, leaveTangent=0.0568, tangentMode=RCTM_User, time=0.4, value=1}
         ]}}}
      CurveIndex=$Emitter.Age
     )
     2=&FloatFromCurve(
      FloatCurve=DataInterface() {propertyValues {Curve {keys [
          {arriveTangent=3.2789, interpMode=RCIM_Cubic, time=0.3, value=0.1}
          {arriveTangent=0.0568, interpMode=RCIM_Cubic, leaveTangent=0.0568, tangentMode=RCTM_User, time=0.5, value=1}
         ]}}}
      CurveIndex=$Emitter.Age
     )
     3=&FloatFromCurve(
      FloatCurve=DataInterface() {propertyValues {Curve {keys [
          {arriveTangent=3.2789, interpMode=RCIM_Cubic, time=0.2, value=0.1}
          {arriveTangent=0.0568, interpMode=RCIM_Cubic, leaveTangent=0.0568, tangentMode=RCTM_User, time=0.4, value=1}
         ]}}}
      CurveIndex=$Emitter.Age
     )
     Index=$User.SwingSectionIndex
     Num=::NewEnumerator2
    )
    Module.Index0Float0Write=false
    Module.Index0Float1Write=true
    Module.Index0Float2Write=false
    Module.Index0Float3Write=false
   )
   /SetVariables(Particles.Swan Color:LinearColor=&MultiplyLinearColor(
     Linear Color=$Particles.Color
     Scale Factor=&MakeVector4FromVectorAndFloat(
      "Vector (RGB)"=<1 1 1>
      "Float (Alpha)"=&VariantFloatSelector(
       Original Float:NiagaraFloat=1
       Variant 1 Float:NiagaraFloat=0
       Variant 2 Float:NiagaraFloat=0
       Variant 3 Float:NiagaraFloat=?
       Variant Amount:NiagaraInt32=1
      )
     )
    )) {isSetParameters=true}
   /SetVariables(Particles.Prime Color:LinearColor=&MultiplyLinearColor(
     Linear Color=$Particles.Color
     Scale Factor=&MakeVector4FromVectorAndFloat(
      "Vector (RGB)"=<1 1 1>
      "Float (Alpha)"=&VariantFloatSelector(
       Original Float:NiagaraFloat=0
       Variant 1 Float:NiagaraFloat=1
       Variant 2 Float:NiagaraFloat=0
       Variant 3 Float:NiagaraFloat=?
       Variant Amount:NiagaraInt32=1
      )
     )
    )) {isSetParameters=true}
   /SetVariables(Particles.Flame Color:LinearColor=&MultiplyLinearColor(
     Linear Color=$Particles.Color
     Scale Factor=&MakeVector4FromVectorAndFloat(
      "Vector (RGB)"=<1 1 1>
      "Float (Alpha)"=&VariantFloatSelector(
       Original Float:NiagaraFloat=0
       Variant 1 Float:NiagaraFloat=0
       Variant 2 Float:NiagaraFloat=1
       Variant 3 Float:NiagaraFloat=?
       Variant Amount:NiagaraInt32=1
      )
     )
    )) {isSetParameters=true}
  ]
  rendererClasses [&NiagaraRibbonRendererProperties]
  renderers [
   {rendererClass=&NiagaraRibbonRendererProperties}
   {rendererClass=&NiagaraRibbonRendererProperties, rendererIndex=1}
   {rendererClass=&NiagaraRibbonRendererProperties, rendererIndex=2}
  ]
 }
}
summary {
 systemName=NS_Pickaxe_CopperMoggy_Trail
 emitters [
  {emitterName=Tech_Trail, rendererClasses [&NiagaraRibbonRendererProperties]}
  {emitterName=Fill_Trail, rendererClasses [&NiagaraRibbonRendererProperties]}
 ]
 userVariables [
  User.RibbonIDOffset {type=NiagaraInt32}
  User.SwingSectionIndex {type=NiagaraInt32}
  User.Variant {type=NiagaraFloat, defaultValue=0, description=""}
 ]
}
