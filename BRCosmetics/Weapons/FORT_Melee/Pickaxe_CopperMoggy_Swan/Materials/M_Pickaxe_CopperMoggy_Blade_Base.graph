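# Unlit master material for the CopperMoggy pickaxe blade, driving an emissive look on skeletal meshes.
# A two-color UV gradient blends with a color-gradient texture, shaped by tech-art masks and two multiplied layers of scrolling Perlin noise.
# Static switches add a digital mosaic glow and flame overrides, both reusing character material functions, before MF_TechArtOverrides.
# With flame enabled, its temporal mask feeds the temporal responsiveness output.
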Material /BRCosmetics/Weapons/FORT_Melee/Pickaxe_CopperMoggy_Swan/Materials/M_Pickaxe_CopperMoggy_Blade_Base
graph {
 properties {
  bAutomaticallySetUsageInEditor=false
  bUseMaterialAttributes=true
  bUsedWithSkeletalMesh=true
  bUsedWithStaticMesh=false
  shadingModel=MSM_Unlit
  usedShadingModels=Unlit
 }
 nodes [
  LinearInterpolate()
  n1 = ComponentMask(None=TextureCoordinate()) {g=true}
  n2 = ScalarParameter() {defaultValue=1, group=CopperMoggy, parameterName=GradientPower}
  n3 = TextureSampleParameter2D(UVs=NamedRerouteUsage()) {
   group=CopperMoggy
   parameterName=Gradient
   texture=/BRCosmetics/Weapons/FORT_Melee/Pickaxe_CopperMoggy_Swan/Skins/Prime/Textures/T_CopperMoggy_Prime_ColorGradient_T_CopperMoggy_Prime_ColorGradient
  }
  n4 = LinearInterpolate(
   A=LinearInterpolate(
    A=VectorParameter() {group=CopperMoggy, parameterName=BaseEmissive_A, defaultValue=<rgba 1 0.0374 0 1>}.RGB
    B=VectorParameter() {group=CopperMoggy, parameterName=BaseEmissive_B, defaultValue=<rgba 1 0.0374 0 1>}.RGB
    Alpha=Power(Base=@n1, Exp=@n2)
   )
   B=Multiply(
    A=NamedRerouteDeclaration(None=@n3.RGB) {name=AnimatedColoredNoise, nodeColor=<rgba 0 1 0.8125 1>}
    B=ScalarParameter() {defaultValue=1, group=CopperMoggy, parameterName=SoftEmissiveIntensity}
   )
   Alpha=NamedRerouteUsage()
  )
  n5 = NamedRerouteUsage()
  n6 = MakeMaterialAttributes!(
   BaseColor
   Metallic
   Specular
   Roughness
   Anisotropy
   EmissiveColor=LinearInterpolate(A=@n4, B=Multiply(A=@n4, B=@n5), Alpha=ScalarParameter() {group=CopperMoggy, parameterName=BorderMaskInfluence})
   Opacity
   OpacityMask
   Normal
   Tangent
   WorldPositionOffset
   SubsurfaceColor
   ClearCoat
   ClearCoatRoughness
   AmbientOcclusion
   Refraction
   CustomizedUVs_0
   CustomizedUVs_1
   CustomizedUVs_2
   CustomizedUVs_3
   CustomizedUVs_4
   CustomizedUVs_5
   CustomizedUVs_6
   CustomizedUVs_7
   PixelDepthOffset
   ShadingModel
   Displacement
  )
  n7 = MaterialFunctionCall!(Material=@n6, Tech Art Mask) {materialFunction=/BRCosmetics/Characters/Player/Female/Medium/Bodies/F_MED_CopperMoggy_Swan/Skins/Prime/Materials/MF_CopperMoggy_DigitalMosaic}
  n8 = GetMaterialAttributes(None=@n7.Material Result) {attributeGetTypes [#1]}
  n9 = Multiply(
   A=Multiply(A=LinearInterpolate(Alpha=@n7.GridMask) {constA=0.1, constB=0.8}, B=NamedRerouteUsage())
   B=ScalarParameter() {group=CopperMoggy, parameterName=AnimatedColorIntensity}
  ) {constB=0}
  n10 = StaticSwitchParameter(
   True=SetMaterialAttributes(MaterialAttributes=@n8.MaterialAttributes, Emissive Color=Add(A=@n8.Emissive Color, B=@n9)) {attributeSetTypes [#1]}
   False=@n6
  ) {group=DigitalMosaic, parameterName=EnableDigitalMosaic}
  n11 = MaterialFunctionCall!(Material=@n10, TechArtMask=NamedRerouteUsage()) {materialFunction=/BRCosmetics/Characters/Player/Female/Medium/Bodies/F_MED_CopperMoggy_Swan/Skins/Flame/Materials/MF_CopperMoggy_FlameOverrides}
  MaterialFunctionCall(In=StaticSwitchParameter(True=@n11.Result, False=@n10) {group=Flame, parameterName=EnableFlameOverrides}) {materialFunction=&MF_TechArtOverrides}
  n12 = TextureSampleParameter2D() {
   group=CopperMoggy
   parameterName=TechArtMask
   samplerSource=SSM_Wrap_WorldGroupSettings
   samplerType=SAMPLERTYPE_Masks
   texture=&T_BlackMask
  }
  NamedRerouteDeclaration(None=@n12.R) {name=EdgeMask, nodeColor=<rgba 0 1 0.9766 1>}
  NamedRerouteDeclaration(None=@n12.G) {name=GradientMask, nodeColor=<rgba 1 0.6563 0 1>}
  NamedRerouteDeclaration(None=@n12.B) {name=BorderMask, nodeColor=<rgba 0.3359 0 1 1>}
  NamedRerouteDeclaration(None=TextureObjectParameter() {group=CopperMoggy, parameterName=SoftNoise, texture=/Game/Global/Textures/Default/Noise/3xPerlin_M}) {name=SoftNoise, nodeColor=<rgba 0.1719 0 1 1>}
  n13 = VectorParameter() {group=CopperMoggy, parameterName=SoftNoiseSpeed_A, defaultValue=<rgba 0.1 0.15 0 0>}
  n14 = Add(
   A=Multiply(A=TextureCoordinate(), B=ScalarParameter() {defaultValue=0.333, group=CopperMoggy, parameterName=SoftNoiseTiling_A})
   B=Multiply(A=AppendVector(A=@n13.R, B=@n13.R), B=Time())
  )
  n15 = NamedRerouteUsage()
  n16 = VectorParameter() {group=CopperMoggy, parameterName=SoftNoiseSpeed_B, defaultValue=<rgba -0.1 0.1 0 0>}
  n17 = Add(
   A=Multiply(A=TextureCoordinate(), B=ScalarParameter() {defaultValue=0.2, group=CopperMoggy, parameterName=SoftNoiseTiling_B})
   B=Multiply(A=AppendVector(A=@n16.R, B=@n16.R), B=Time())
  )
  n18 = NamedRerouteUsage()
  n19 = Multiply(
   A=Multiply(
    A=TextureSample(UVs=@n14, Tex=@n15) {samplerSource=SSM_Wrap_WorldGroupSettings, texture=&3xPerlin_M}.R
    B=TextureSample(UVs=@n17, Tex=@n18) {samplerSource=SSM_Wrap_WorldGroupSettings, texture=&3xPerlin_M}.G
   )
   B=ScalarParameter() {defaultValue=2, group=CopperMoggy, parameterName=SoftNoiseScale}
  ) {constB=2}
  n20 = ScalarParameter() {defaultValue=1, group=CopperMoggy, parameterName=SoftNoisePower}
  NamedRerouteDeclaration(None=Saturate(None=Power(Base=@n19, Exp=@n20))) {name=AnimatedSoftNoise, nodeColor=<rgba 0.1484 1 0 1>}
  NamedRerouteDeclaration(None=@n12.RGB) {name=TechArtMask, nodeColor=<rgba 1 0.7969 0 1>}
  TemporalResponsivenessOutput!(None=StaticSwitchParameter(True=@n11.Temporal Mask, False=Constant()) {group=Flame, parameterName=EnableFlameOverrides})
 ]
}
